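# Material function that outputs a rounded box mask for UI materials, with inputs for UVs, hardness, corner roundness, scale from center and aspect ratio.
# It scales and mirrors UVs around the center, then normalizes edge distance by roundness and clamps it, so a single SphereMask draws all four rounded corners.
# When Keep Aspect Ratio is at least 1, the corner size is corrected by the Max Ratio from MF_UGCUI_GetAspectRatio.
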MaterialFunction /FortUILibrary/UGC/MaterialFunctions/MF_UGCUI_Box_Cheaper
graph {nodes [
  Constant2Vector() {g=1, r=1}
  Constant2Vector() {g=1, r=1}
  Constant() {r=0.25}
  Constant()
  Constant() {r=1}
  n1 = FunctionInput(Preview=TextureCoordinate()) {bUsePreviewValueAsDefault=true, inputName=UVs, inputType=FunctionInput_Vector2}
  n2 = Constant() {r=0.5}
  n3 = FunctionInput() {bUsePreviewValueAsDefault=true, inputName=UV Scale from center, inputType=FunctionInput_Vector2, sortPriority=3, previewValue=<1 1 0 1>}
  n4 = Add(A=Multiply(A=Subtract(A=@n1, B=@n2), B=@n3), B=@n2)
  n5 = Multiply(A=Abs(None=Subtract(A=@n4) {constB=0.5})) {constB=2}
  n6 = FunctionInput() {bUsePreviewValueAsDefault=true, inputName=Corner Roundness, inputType=FunctionInput_Scalar, sortPriority=2, previewValue=<0.25 0 0 1>}
  n7 = FunctionInput() {bUsePreviewValueAsDefault=true, inputName=Aspect Ratio, inputType=FunctionInput_Vector2, sortPriority=4, previewValue=<1 1 0 1>}
  n8 = FunctionInput() {
   bUsePreviewValueAsDefault=true
   description="Let the box stretch < 1 <= Keep your (rounded) box aspect ratio "
   inputName=Keep Aspect Ratio
   inputType=FunctionInput_Scalar
   sortPriority=5
   previewValue=<1 0 0 1>
  }
  n9 = LinearInterpolate(
   A=Constant() {r=1}
   B=MaterialFunctionCall() {materialFunction=&/MF_UGCUI_GetAspectRatio}.Max Ratio
   Alpha=Saturate(None=Floor(None=@n8))
  )
  n10 = Divide(A=@n6, B=Multiply(A=@n7, B=NamedRerouteDeclaration(None=@n9) {name=Ratio, nodeColor=<rgba 0 0 0 1>}))
  n11 = Saturate(None=Multiply(A=OneMinus(None=@n5), B=Divide(B=@n10)) {constB=3})
  n12 = Constant() {r=0.5}
  FunctionOutput(None=SphereMask(
    A=Divide(A=@n11)
    B=@n12
    Radius=@n12
    Hardness=FunctionInput() {bUsePreviewValueAsDefault=true, inputName=Hardness, inputType=FunctionInput_Scalar, sortPriority=1}
   ))
 ]}
