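# UGC UI material function that anti-aliases a distance field gradient around a Threshold input, 0.5 by default.
# It takes screen-space DDX and DDY of the field's length and scales their magnitude into an edge width that adapts to pixel footprint.
# SmoothStep produces two outputs: EitherSide centers the transition on the threshold, while Inner places it just below, with reversed bounds that invert the ramp.
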MaterialFunction /FortUILibrary/UGC/MaterialFunctions/MF_UGCUI_AA
graph {
 properties {description=Adjust the aliasing of a distance field gradient., libraryCategoriesText [UI Library]}
 nodes [
  n1 = FunctionInput(Preview=Constant() {r=0.5}) {
   bUsePreviewValueAsDefault=true
   description="Define the aliasing threshold (Default : 0.5)"
   inputName=Threshold
   inputType=FunctionInput_Scalar
  }
  n2 = FunctionInput(Preview=TextureCoordinate()) {
   bUsePreviewValueAsDefault=true
   description="Set the distance field gradient. If you are sending an SDF value, add 0.5 to your value."
   inputName=DistanceField
   inputType=FunctionInput_Scalar
  }
  MaterialExpressionNamedRerouteDeclaration_0 = NamedRerouteDeclaration(None=@n2) {name=Distance Field, nodeColor=<rgba 0 0 0 1>}
  n3 = Custom!(Vec=@MaterialExpressionNamedRerouteDeclaration_0) {
   code="return length(Vec);"
   description=Length
   outputType=CMOT_Float1
   inputs [{inputName=Vec, input {expression=&self:MaterialExpressionNamedRerouteDeclaration_0}}]
  }
  MaterialExpressionAppendVector_0 = AppendVector(A=DDX(Value=@n3), B=DDY(Value=@n3))
  n4 = Custom(Vec=@MaterialExpressionAppendVector_0) {
   code="return length(Vec);"
   description=Length
   outputType=CMOT_Float1
   inputs [{inputName=Vec, input {expression=&self:MaterialExpressionAppendVector_0}}]
  }
  n5 = Multiply(A=@n4, B=Constant() {r=1.5})
  n6 = Multiply(A=@n5) {constB=0.5}
  FunctionOutput(None=SmoothStep(Min=Subtract(A=@n1, B=@n6), Max=Add(A=@n1, B=@n6), Value=NamedRerouteUsage())) {outputName=EitherSide}
  FunctionOutput(None=SmoothStep(Min=@n1, Max=Subtract(A=@n1, B=@n5), Value=NamedRerouteUsage())) {outputName=Inner}
 ]
}
