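# MF_UGCUI_IsolateQuads is a library material function for UGC UI that splits UV space into four quads meant to carry independent values from a Vector4.
# Centered UVs pass through Ceil and Saturate to form branchless X/Y quadrant masks.
# These masks drive nested lerps selecting left/right, then top/bottom.
# As wired, the left pair duplicates R and the right duplicates B, leaving G and A unused.
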MaterialFunction /FortUILibrary/UGC/MaterialFunctions/MF_UGCUI_IsolateQuads
graph {
 properties {bExposeToLibrary=true, description=This function allow you to isolate 4 quads and control each of them independently between 0 and 1.}
 nodes [
  n1 = NamedRerouteUsage() {desc="where 0 = topQuads\r\nwhere 1 = bottomQuads"}
  n2 = LinearInterpolate(A=NamedRerouteUsage(), B=NamedRerouteUsage(), Alpha=NamedRerouteUsage() {desc="where 0 = leftQuads\r\nwhere 1 = rightQuads"}) {constAlpha=0.0027, desc="are we left or right?\r\nApply Left quads values to 0\r\nApply Right quads values to 1"}
  n3 = ComponentMask(None=@n2) {desc=Top QUads Values, r=true}
  n4 = ComponentMask(None=@n2) {desc=Bottom Quads Values, g=true}
  FunctionOutput(None=NamedRerouteDeclaration(None=LinearInterpolate(A=@n3, B=@n4, Alpha=@n1) {constAlpha=0.2427, desc="are we top or bottom?\r\nApply Left quads values to 0\r\nApply Right quads values to 1"}) {name=Split Quads, nodeColor=<rgba 0 0 0 1>}) {outputName=Isolated 4-Quad, sortPriority=3}
  n5 = FunctionInput(Preview=Constant4Vector() {constant=<rgba 0.2 0.1 0.3 0.5>}.RGBA) {
   bUsePreviewValueAsDefault=true
   description=The value between 0 and 1 of the 4 separated quads of the quadrant. Use a vector4 to set your values.
   inputName=4-Quad Values
   inputType=FunctionInput_Vector4
   sortPriority=1
  }
  n6 = MaterialFunctionCall(Float4=@n5) {materialFunction=&BreakOutFloat4Components}
  NamedRerouteDeclaration(None=AppendVector(A=@n6.B, B=@n6.B)) {name=Right Quads, nodeColor=<rgba 0 0 0 1>}
  n7 = Saturate(None=Ceil(None=NamedRerouteUsage()))
  NamedRerouteDeclaration(None=ComponentMask(None=@n7) {r=true}) {name=quadrant X, nodeColor=<rgba 0 0 0 1>}
  NamedRerouteDeclaration(None=ComponentMask(None=@n7) {g=true}) {name=quadrant Y, nodeColor=<rgba 0 0 0 1>}
  NamedRerouteDeclaration(None=AppendVector(A=@n6.R, B=@n6.R)) {name=Left Quads, nodeColor=<rgba 0 0 0 1>}
  n8 = TextureCoordinate()
  NamedRerouteDeclaration(None=ConstantBiasScale(None=FunctionInput(Preview=@n8) {bUsePreviewValueAsDefault=true, description=UVs, inputName=UVs, inputType=FunctionInput_Vector2}) {bias=-0.5, scale=2}) {name=Centered UVs, nodeColor=<rgba 0 0 0 1>}
  Reroute() {desc="https://www.notion.so/fortnite-uxui/IsolateQuads-b663380126a0476a8aed3c0d34f6a6c0?pvs=4"}
 ]
}
