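# UGC UI material function that draws a rounded-box SDF from UVs, outputting fill, inside-clipped outline and both raw distance fields.
# Centered, mirrored UVs are offset by Size, with optional aspect-ratio correction and per-corner roundness via MF_UGCUI_IsolateQuads.
# Distance combines a Custom-node length with an interior term that, as wired, reads only the R component.
# Outline derives from absolute distance minus thickness; SmoothStep with Glow Min/Max inputs controls falloff.
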MaterialFunction /FortUILibrary/UGC/MaterialFunctions/MF_UGCUI_Box
graph {
 properties {description=SDF Box UGC Function, libraryCategoriesText [""]}
 nodes [
  FunctionOutput(None=NamedRerouteUsage()) {outputName=SDF, sortPriority=3}
  FunctionOutput(None=NamedRerouteUsage()) {outputName=SDF Outline, sortPriority=4}
  n1 = OneMinus(None=SmoothStep(Min=NamedRerouteUsage(), Max=NamedRerouteUsage(), Value=NamedRerouteUsage()))
  FunctionOutput(None=@n1) {outputName=Fill}
  n2 = SmoothStep(Min=NamedRerouteUsage(), Max=NamedRerouteUsage(), Value=NamedRerouteUsage())
  FunctionOutput(None=Min(A=@n1, B=OneMinus(None=@n2)) {desc=Keep the outline inside the shape}) {outputName=Outline, sortPriority=2}
  n3 = FunctionInput(Preview=TextureCoordinate()) {bUsePreviewValueAsDefault=true, description=UVs, inputName=UVs, inputType=FunctionInput_Vector2}
  n4 = FunctionInput(Preview=Constant4Vector().RGBA) {
   bUsePreviewValueAsDefault=true
   description="The radius of the rounded corners. Use a Float (Scalar) for Uniform Roundness, or a Float4 to set a different radius on any corner. Float4 requires to set Uniform Corner Roundness to FALSE."
   inputName=Corner Roundness
   inputType=FunctionInput_Vector4
   sortPriority=6
   previewValue=<0 0 0 0>
  }
  NamedRerouteDeclaration(None=MaterialFunctionCall!(UVs=@n3, 4-Quad Values=@n4) {desc=computed on Uniform Corner Roundness, materialFunction=&/MF_UGCUI_IsolateQuads}.Isolated 4-Quad) {name=cornersRoundness, nodeColor=<rgba 0 0 0 1>}
  MaterialExpressionMax_0 = Max(A=NamedRerouteUsage()) {constB=0}
  n5 = Custom!(d=@MaterialExpressionMax_0) {
   code="length(d);"
   description=Length
   outputType=CMOT_Float1
   inputs [{inputName=d, input {expression=&self:MaterialExpressionMax_0}}]
  }
  n6 = ConstantBiasScale(None=@n3) {bias=-0.5, scale=2}
  n7 = FunctionInput(Preview=Constant2Vector() {g=0.5, r=0.5}) {
   bUsePreviewValueAsDefault=true
   description="Box Size. Default value: 0.5, 0.5"
   inputName=Size
   inputType=FunctionInput_Vector2
   sortPriority=1
   previewValue=<0.5 0.5 0 1>
  }
  n8 = Multiply(A=Subtract(A=Abs(None=@n6), B=@n7), B=NamedRerouteUsage())
  n9 = NamedRerouteUsage()
  n10 = MaterialFunctionCall(Float2=NamedRerouteDeclaration(None=Add(A=@n8, B=@n9)) {name=Box UVs, nodeColor=<rgba 0 0 0 1>}) {materialFunction=&BreakOutFloat2Components}
  n11 = Min(A=Max(A=@n10.R, B=@n10.R) {constB=0}) {constB=0}
  NamedRerouteDeclaration(None=Subtract(A=Add(A=@n5, B=@n11), B=NamedRerouteUsage())) {name=SDF Shape, nodeColor=<rgba 0 0 0 1>}
  n12 = NamedRerouteUsage()
  n13 = FunctionInput(Preview=Constant() {r=0.05}) {
   bUsePreviewValueAsDefault=true
   description=Determines the thickness of the stroke when using one of the outline outputs.
   inputName=Outline Thickness
   inputType=FunctionInput_Scalar
   sortPriority=2
  }
  NamedRerouteDeclaration(None=Subtract(A=Abs(None=@n12), B=@n13) {constB=0.1}) {name=SDF Outline, nodeColor=<rgba 0 0 0 1>}
  n14 = FunctionInput(Preview=Constant()) {
   bUsePreviewValueAsDefault=true
   description=Determines glow falloff strength. Can not go below GlowMin value or results will invert
   inputName=Glow Min
   inputType=FunctionInput_Scalar
   sortPriority=3
  }
  NamedRerouteDeclaration(None=@n14) {name=Glow Min, nodeColor=<rgba 0 0 0 1>}
  n15 = FunctionInput(Preview=Constant()) {
   bUsePreviewValueAsDefault=true
   description=Determines glow falloff strength. Can not go below GlowMin value or results will invert
   inputName=Glow Max
   inputType=FunctionInput_Scalar
   sortPriority=4
  }
  NamedRerouteDeclaration(None=@n15) {name=Glow Max, nodeColor=<rgba 0 0 0 1>}
  n16 = FunctionInput(Preview=Constant() {r=1}) {
   bUsePreviewValueAsDefault=true
   description="Let the box stretch < 1 <= Keep your (rounded) box aspect ratio "
   inputName=Keep Aspect Ratio
   inputType=FunctionInput_Scalar
   sortPriority=7
   previewValue=<1 0 0 1>
  }
  n17 = Floor(None=@n16)
  NamedRerouteDeclaration(None=LinearInterpolate(
    A=Constant() {r=1}
    B=MaterialFunctionCall() {materialFunction=&/MF_UGCUI_GetAspectRatio}.Max Ratio
    Alpha=Saturate(None=@n17)
   )) {name=Ratio, nodeColor=<rgba 0 0 0 1>}
 ]
}
