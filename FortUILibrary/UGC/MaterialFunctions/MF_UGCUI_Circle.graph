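# UI shader library material function that draws a procedural circle from a signed distance field, with fill and outline outputs.
# It remaps the UV input to a centered -1 to 1 range and subtracts Size from the distance to center.
# The outline SDF is the absolute circle SDF minus OutlineThickness.
# Inverted SmoothSteps between GlowMin and GlowMax turn both SDFs into soft-edged masks, and both raw SDFs are also output.
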MaterialFunction /FortUILibrary/UGC/MaterialFunctions/MF_UGCUI_Circle
graph {
 properties {bExposeToLibrary=true, description=Circle Function, libraryCategoriesText [UI_ShaderLibrary]}
 nodes [
  Constant() {r=0.5}
  Constant() {r=0.05}
  Constant()
  Constant()
  n1 = FunctionInput() {
   bUsePreviewValueAsDefault=true
   description=Determines glow falloff strength. Should not exceed GlowMax or results will invert
   inputName=GlowMin
   inputType=FunctionInput_Scalar
   sortPriority=5
  }
  n2 = FunctionInput() {
   bUsePreviewValueAsDefault=true
   description=Determines glow falloff strength. Can not go below GlowMin value or results will invert
   inputName=GlowMax
   inputType=FunctionInput_Scalar
   sortPriority=6
  }
  n3 = FunctionInput(Preview=TextureCoordinate()) {
   bUsePreviewValueAsDefault=true
   description=Controls the UVs of the circle with the TextureCoordinate node plugged in. Can be use to manipulate the X/Y coordinates as well.
   inputName=UV
   inputType=FunctionInput_Vector2
   sortPriority=1
  }
  n4 = FunctionInput() {
   bUsePreviewValueAsDefault=true
   description=Determines the size of the circle. A '1' fills the circle to the border with no glow or outline added.
   inputName=Size
   inputType=FunctionInput_Scalar
   sortPriority=2
   previewValue=<0.5 0 0 1>
  }
  n5 = Subtract(A=Distance(A=ConstantBiasScale(None=@n3) {bias=-0.5, scale=2}, B=Constant()), B=@n4)
  FunctionOutput(None=OneMinus(None=SmoothStep(Min=@n1, Max=@n2, Value=@n5))) {outputName=Fill, sortPriority=1}
  FunctionOutput(None=@n5) {outputName=SDF, sortPriority=3}
  n6 = FunctionInput() {
   bUsePreviewValueAsDefault=true
   description=Determines the thickness of the stroke when using one of the outline outputs.
   inputName=OutlineThickness
   inputType=FunctionInput_Scalar
   sortPriority=7
   previewValue=<0.05 0 0 1>
  }
  n7 = Subtract(A=Abs(None=@n5), B=@n6)
  FunctionOutput(None=OneMinus(None=SmoothStep(Min=@n1, Max=@n2, Value=@n7))) {outputName=Outline, sortPriority=2}
  FunctionOutput(None=@n7) {outputName=SDFOutline, sortPriority=4}
 ]
}
