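# UGC UI material function, exposed to the library, that turns a 0-1 scalar into nine ease curves: In, Out and InOut for Sine, Cubic and Quint.
# The input feeds all curves via a named reroute and defaults to looping Frac(Time) when unconnected.
# Sine curves use trig nodes, while Cubic and Quint chain multiplies.
# Their InOut variants select halves with a rounded Lerp alpha, and every output is saturated.
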MaterialFunction /FortUILibrary/UGC/MaterialFunctions/MF_UGCUI_EaseCurves
graph {
 properties {bExposeToLibrary=true, description=Outputs interpolated values using ease curves.}
 nodes [
  n1 = NamedRerouteUsage()
  n2 = Multiply(A=@n1, B=@n1)
  FunctionOutput(None=Saturate(None=Multiply(A=@n2, B=@n1))) {description=Ouputs an EaseInCubic curve, outputName=EaseInCubic, sortPriority=3}
  n3 = OneMinus(None=NamedRerouteUsage())
  n4 = Multiply(A=Multiply(A=@n3, B=@n3), B=@n3)
  FunctionOutput(None=Saturate(None=OneMinus(None=@n4))) {description=Ouputs an EaseOutCubic curve, outputName=EaseOutCubic, sortPriority=4}
  n5 = Subtract(A=Cosine(None=NamedRerouteUsage()) {period=2})
  FunctionOutput(None=Saturate(None=Multiply(A=@n5) {constB=-0.5})) {description=Ouputs an EaseInOutSine curve, outputName=EaseInOutSine, sortPriority=2}
  n6 = Cosine(None=Multiply(A=NamedRerouteUsage()) {constB=0.5}) {period=2}
  FunctionOutput(None=Saturate(None=OneMinus(None=@n6))) {description=Ouputs an EaseInSine curve, outputName=EaseInSine}
  n7 = Multiply(A=NamedRerouteUsage()) {constB=0.5}
  FunctionOutput(None=Saturate(None=Sine(None=@n7) {period=2})) {description=Ouputs an EaseOutSine curve, outputName=EaseOutSine, sortPriority=1}
  n8 = Constant() {r=16}
  n9 = NamedRerouteUsage()
  n10 = OneMinus(None=@n9)
  n11 = Multiply(A=@n10, B=@n10)
  n12 = Multiply(A=@n11, B=@n11)
  n13 = OneMinus(None=Multiply(A=@n8, B=Multiply(A=@n12, B=@n10)))
  n14 = OneMinus(None=Round(None=NamedRerouteUsage()))
  n15 = Multiply(A=@n9, B=@n9)
  n16 = Multiply(A=@n8, B=Multiply(A=Multiply(A=@n15, B=@n15), B=@n9))
  FunctionOutput(None=Saturate(None=LinearInterpolate(A=@n13, B=@n16, Alpha=@n14))) {description=Ouputs an EaseInOutQuint curve, outputName=EaseInOutQuint, sortPriority=8}
  n17 = NamedRerouteUsage()
  n18 = Multiply(A=@n17, B=@n17)
  n19 = Multiply(A=@n18, B=@n18)
  FunctionOutput(None=Saturate(None=Multiply(A=@n19, B=@n17))) {description=Ouputs an EaseInQuint curve, outputName=EaseInQuint, sortPriority=6}
  n20 = OneMinus(None=NamedRerouteUsage())
  n21 = Multiply(A=@n20, B=@n20)
  n22 = Multiply(A=Multiply(A=@n21, B=@n21), B=@n20)
  FunctionOutput(None=Saturate(None=OneMinus(None=@n22))) {description=Ouputs an EaseOutQuint curve, outputName=EaseOutQuint, sortPriority=7}
  n23 = Constant() {r=4}
  n24 = NamedRerouteUsage()
  n25 = OneMinus(None=@n24)
  n26 = Multiply(A=@n23, B=@n25) {constB=0.5}
  n27 = OneMinus(None=Multiply(A=Multiply(A=@n26, B=@n25), B=@n25))
  n28 = OneMinus(None=Round(None=NamedRerouteUsage()))
  n29 = Multiply(A=Multiply(A=Multiply(A=@n23, B=@n24) {constB=0.5}, B=@n24), B=@n24)
  FunctionOutput(None=Saturate(None=LinearInterpolate(A=@n27, B=@n29, Alpha=@n28))) {description=Ouputs an EaseInOutCubic curve, outputName=EaseInOutCubic, sortPriority=5}
  n30 = FunctionInput(Preview=Frac(None=Time())) {bUsePreviewValueAsDefault=true, description=Inputs a value between 0-1., inputName="Value (0-1)", inputType=FunctionInput_Scalar}
  NamedRerouteDeclaration(None=@n30) {name=Value, nodeColor=<rgba 0 0 0 1>}
 ]
}
