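# UI material function, exposed in the UI Library, that skews UVs by a two-axis angle and defaults to the 9-slice user interface UV.
# It takes the tangent of each angle in degrees and uses swizzles to offset each UV axis in proportion to the other.
# Subtracting half the tangents keeps the shear centered on the 0.5 midpoint of the UVs.
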MaterialFunction /FortUILibrary/UGC/MaterialFunctions/MF_UGCUI_Skew
graph {
 properties {bExposeToLibrary=true, description=Allows you to skew the UVs, libraryCategoriesText [UI Library]}
 nodes [
  n1 = NamedRerouteUsage()
  n2 = FunctionInput(Preview=Constant2Vector() {r=45}) {
   bUsePreviewValueAsDefault=true
   description=Define the skew angle on 2 axis
   inputName=Skew Angle
   inputType=FunctionInput_Vector2
   sortPriority=2
  }
  n3 = MaterialFunctionCall(Float2=@n2) {materialFunction=&BreakOutFloat2Components}
  n4 = AppendVector(A=Tangent!(None=@n3.R) {period=360}, B=Tangent(None=@n3.G) {period=360})
  n5 = MaterialFunctionCall(XY=Multiply(A=@n1, B=MaterialFunctionCall(XY=@n4) {materialFunction=&Swizzle}.YX)) {materialFunction=&Swizzle}
  FunctionOutput(None=Subtract(A=Add(A=@n1, B=@n5.YX), B=Multiply(A=@n4) {constB=0.5})) {description=Returns the skewed version of your initial UVs, outputName=Skewed UVs}
  NamedRerouteDeclaration(None=FunctionInput(Preview=MaterialFunctionCall() {materialFunction=&GetUserInterfaceUV}.9-Slice UV) {bUsePreviewValueAsDefault=true, description=Define the UVs, inputName=UV, inputType=FunctionInput_Vector2}) {name=UV, nodeColor=<rgba 0 0 0 1>}
 ]
}
